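# Physics asset for the Nautilus two-handed pickaxe skeletal mesh, a cosmetic melee weapon.
# It holds a single capsule body on the Melee_TwoHanded_Handle bone, offset and rotated to cover the weapon as a whole.
# It defines no constraints, so the pickaxe collides as one piece with no jointed parts.
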PhysicsAsset /BRCosmetics/Weapons/FORT_Melee/Pickaxe_Nautilus/Meshes/SK_Pickaxe_Nautilus_PhysicsAsset
bodies {Melee_TwoHanded_Handle [{shapeType=Capsule, length=71.5433, radius=41.4397, center=<0.1088 7.7705 26.8432>, rotation=<rot -2.3857 -90 -0.1135>}]}
constraints []
